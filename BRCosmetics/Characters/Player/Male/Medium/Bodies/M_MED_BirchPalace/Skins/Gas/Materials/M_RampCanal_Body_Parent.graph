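# Parent cloth material for the BirchPalace "Gas" (RampCanal) character body, adding a glowing inner-smoke emissive look to the 2023 anime shader.
# Camera-aligned virtual planes drive panned noise and sphere masks for chest and upper-body glow; a box mask on pre-skinned position isolates the arms.
# TA mask textures and a glider flowing-smoke function shape the detail.
# Fresnel darkens and tints the emissive, while lit and shaded colors blend toward a weighted layer.
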Material /BRCosmetics/Characters/Player/Male/Medium/Bodies/M_MED_BirchPalace/Skins/Gas/Materials/M_RampCanal_Body_Parent
graph {
 properties {
  bAutomaticallySetUsageInEditor=false
  bUseMaterialAttributes=true
  bUsedWithMeshDeformer=true
  bUsedWithMorphTargets=true
  bUsedWithSkeletalMesh=true
  pixelDepthOffsetMode=PDOM_Legacy
  shadingModel=MSM_Cloth
  usedShadingModels=Cloth
 }
 nodes [
  n1 = MaterialFunctionCall!(CustomUV) {materialFunction=/Game/Global/Materials/Character_Parent_2023/Anime_Parent_2023/MaterialFunctions/MFC_AnimeShader_2023_Pre}
  n2 = MaterialFunctionCall!(CustomLightVector, OverrideNormal_WS, STT=@n1.STT, Normal_Map=@n1.STT, DFL=@n1.STT, FX=@n1.STT) {materialFunction=/Game/Global/Materials/Character_Parent_2023/Anime_Parent_2023/MaterialFunctions/MFC_AnimeShader_2023_Calculations}
  n3 = NamedRerouteUsage()
  n4 = ScalarParameter() {defaultValue=0.3, group=RampCanal, parameterName=BaseColor_BlendWeight}
  n5 = NamedRerouteUsage()
  MaterialFunctionCall!(
   Shading_Gradient=@n2.Shading_Gradient
   Threshold_Gradient=@n2.Shading_Gradient
   Specular_Gradient=@n2.Shading_Gradient
   Sheen_Mask=@n2.Shading_Gradient
   Thickness=@n2.Shading_Gradient
   DFL_Mask=@n2.Shading_Gradient
   FX=@n2.Shading_Gradient
   Metallic_Gradient=@n2.Shading_Gradient
   RimLight_Gradient=@n2.Shading_Gradient
   NormalBuffer=@n2.Shading_Gradient
   --------
   CINE_LightSources=@n2.Shading_Gradient
   CINE_LightColor=@n2.Shading_Gradient
   ---------
   Color_Lit=LinearInterpolate(A=NamedRerouteUsage(), B=Multiply(A=@n3, B=@n4) {constB=0.2}, Alpha=NamedRerouteUsage()) {constB=0}
   Color_Shaded=LinearInterpolate(A=NamedRerouteUsage(), B=Multiply(A=@n5, B=@n4) {constB=0.2}, Alpha=NamedRerouteUsage()) {constB=0}
   SRM=@n1.SRM
   Emissive=NamedRerouteUsage()
   DFL_Color=NamedRerouteDeclaration(None=@n1.DFL_Color) {name=DFL_Color, nodeColor=<rgba 1 0 0.9844 1>}
   Specular_Color=@n1.SRM
   RimLight_Color=@n1.SRM
  ) {materialFunction=/Game/Global/Materials/Character_Parent_2023/Anime_Parent_2023/MaterialFunctions/MFC_AnimeShader_2023_Post}
  NamedRerouteDeclaration(None=@n1.Color_Shaded) {name=Color_Shaded, nodeColor=<rgba 1 0.9141 0 1>}
  n6 = NamedRerouteUsage()
  n7 = NamedRerouteUsage()
  n8 = OneMinus(None=Saturate(None=NamedRerouteUsage()))
  n9 = MaterialFunctionCall() {materialFunction=&/MFW_CameraAlignedVirtualPlane}
  n10 = AppendVector(A=Constant() {r=0.5}, B=Constant() {r=0.096})
  n11 = Constant() {r=0.5}
  n12 = Constant() {r=0.7155}
  n13 = Add(A=Multiply(A=SphereMask(A=@n9.UV, B=@n10, Radius=@n11, Hardness=@n12), B=Constant() {r=0.2}) {constB=0.5}, B=NamedRerouteUsage())
  NamedRerouteDeclaration(None=Max(A=Multiply(A=@n6, B=@n7), B=Multiply(A=@n8, B=@n13))) {name=InnerGlowMask, nodeColor=<rgba 0 1 0.2734 1>}
  NamedRerouteDeclaration(None=@n1.Color_Lit) {name=Color_Lit, nodeColor=<rgba 0.6641 1 0 1>}
  n14 = Constant() {r=1}
  n15 = Add(A=NamedRerouteUsage(), B=Constant() {r=0.2}) {constB=0.15}
  n16 = Multiply(A=NamedRerouteUsage()) {constB=1.5}
  n17 = LinearInterpolate(
   A=ScalarParameter() {defaultValue=0.75, group=RampCanal, parameterName=Noise_EmissiveIntensity_Min}
   B=LinearInterpolate(
    A=Constant() {r=1}
    B=ScalarParameter() {defaultValue=50, group=RampCanal, parameterName=Noise_EmissiveIntensity_Max}
    Alpha=Saturate(None=@n15)
   )
   Alpha=Saturate(None=Power(Base=@n16))
  ) {constA=1, constB=2}
  n18 = ScalarParameter() {group=RampCanal, parameterName=EmissiveDetailMaskIntensity_MIn}
  n19 = ScalarParameter() {defaultValue=3, group=RampCanal, parameterName=EmissiveDetailMaskIntensity_Max}
  n20 = NamedRerouteUsage()
  NamedRerouteDeclaration(None=Multiply(A=Multiply(A=@n14, B=@n17), B=LinearInterpolate(A=@n18, B=@n19, Alpha=@n20) {constA=1, constB=2})) {name=RampCanalBaseSmoke, nodeColor=<rgba 0 0.9062 1 1>}
  NamedRerouteDeclaration(None=@n1.Emissive) {name=In_Emissive, nodeColor=<rgba 1 0.7734 0 1>}
  NamedRerouteDeclaration(None=ComponentMask(None=@n1.STT) {g=true}) {name=AO, nodeColor=<rgba 0.1719 1 0 1>}
  n21 = ScalarParameter() {defaultValue=0.1, group=RampCanal, parameterName=FireSmokelEmissiveIntensity_Min}
  n22 = ScalarParameter() {defaultValue=0.09, group=RampCanal, parameterName=FireSmokelEmissiveIntensity_Min2}
  n23 = NamedRerouteUsage()
  n24 = Saturate(None=Max(A=NamedRerouteUsage(), B=Multiply(A=@n23) {constB=0.5}))
  n25 = Constant() {r=0.5}
  n26 = ScalarParameter() {defaultValue=3.5, group=RampCanal, parameterName=FireSmokeEmissiveIntensity_Max}
  n27 = MaterialFunctionCall!(In_FireSmoke_TAMask=NamedRerouteUsage(), In_FireSmoke_UV=NamedRerouteUsage()) {materialFunction=/BRCosmetics/Gadgets/Assets/VinderTech_GliderChute/Glider_RampCanal/Materials/MFW_FlowingSmoke}
  n28 = LinearInterpolate(Alpha=NamedRerouteUsage()) {constA=0.1}
  NamedRerouteDeclaration(None=LinearInterpolate(
    A=LinearInterpolate(A=@n21, B=@n22, Alpha=@n24)
    B=LinearInterpolate(A=@n25, B=@n26, Alpha=@n24) {constA=1}
    Alpha=Multiply(A=@n27.Result, B=@n28)
   ) {constA=1, constB=2}) {name=DetailSmoke, nodeColor=<rgba 0 1 0.1328 1>}
  n29 = NamedRerouteUsage()
  n30 = Multiply(
   A=LinearInterpolate(
    A=VectorParameter() {group=RampCanal, parameterName=EmissiveColorBase01, defaultValue=<rgba 0.09 0.1267 0.2 1>}.RGB
    B=VectorParameter() {group=RampCanal, parameterName=EmissiveColorInternal, defaultValue=<rgba 0.081 0.2878 0.8 1>}.RGB
    Alpha=Saturate(None=@n29)
   )
   B=LinearInterpolate(A=Constant() {r=10}, B=NamedRerouteUsage(), Alpha=NamedRerouteUsage()) {constA=10}
  )
  n31 = LinearInterpolate(Alpha=NamedRerouteUsage()) {constA=1, constB=0.05}
  n32 = Multiply(A=LinearInterpolate(A=NamedRerouteUsage(), B=Multiply(A=@n30, B=@n31), Alpha=NamedRerouteUsage()), B=NamedRerouteUsage())
  n33 = Constant() {r=1}
  n34 = Constant() {r=-0.5}
  n35 = Saturate(None=Fresnel(ExponentIn=Constant() {r=1}) {exponent=1})
  n36 = NamedRerouteDeclaration(None=Saturate(None=LinearInterpolate(A=@n33, B=@n34, Alpha=@n35))) {name=Fresnel, nodeColor=<rgba 1 0.0703 0 1>}
  n37 = NamedRerouteUsage()
  n38 = TextureSample() {
   samplerSource=SSM_Wrap_WorldGroupSettings
   samplerType=SAMPLERTYPE_Masks
   texture=/BRCosmetics/Characters/Player/Male/Medium/Bodies/M_MED_BirchPalace/Skins/Gas/Textures/T_BirchPalace_Gas_Body_TA2
  }
  n39 = Multiply(
   A=LinearInterpolate(A=@n32, B=Multiply(A=@n32, B=@n36), Alpha=ScalarParameter() {defaultValue=0.75, parameterName=FresnelBrightness})
   B=LinearInterpolate(B=Constant() {r=15}, Alpha=Multiply(A=@n37, B=@n38.R)) {constA=1, constB=3}
  )
  n40 = Constant4Vector() {constant=<rgba 0.0331 0.0742 0.1981 1>}
  n41 = LinearInterpolate(A=Multiply(A=Constant() {r=0.3}, B=NamedRerouteUsage()), Alpha=Saturate(None=NamedRerouteUsage())) {constB=0.05}
  n42 = Max(A=OneMinus(None=NamedRerouteUsage()), B=NamedRerouteUsage())
  n43 = NamedRerouteUsage()
  NamedRerouteDeclaration(None=LinearInterpolate(
    A=VectorParameter() {group=RampCanal, parameterName=FresnelGlowColor, defaultValue=<rgba 0.0502 0.0537 0.22 1>}.RGB
    B=LinearInterpolate(A=@n39, B=@n40.RGB, Alpha=@n41)
    Alpha=LinearInterpolate(B=@n42, Alpha=@n43) {constA=1}
   )) {name=Out_Emissive, nodeColor=<rgba 1 0 0.1875 1>}
  n44 = Constant() {r=0.5}
  n45 = Constant() {r=0.244}
  NamedRerouteDeclaration(None=SphereMask(
    A=MaterialFunctionCall() {materialFunction=&MFW_CameraAlignedVirtualPlane}.UV
    B=AppendVector(A=@n44, B=@n45)
    Radius=Constant() {r=0.64}
    Hardness=Constant() {r=0.1}
   )) {name=UpperBodyMask, nodeColor=<rgba 0 0.7187 1 1>}
  n46 = MaterialFunctionCall() {materialFunction=&MFW_CameraAlignedVirtualPlane}
  n47 = Constant() {r=5}
  n48 = TextureSample(UVs=Panner(Coordinate=Multiply(A=@n46.UV, B=@n47) {constB=11}, Speed=Constant() {r=0.1}) {speedX=0.07, speedY=0.07}) {
   samplerSource=SSM_Wrap_WorldGroupSettings
   samplerType=SAMPLERTYPE_Grayscale
   texture=/Game/Global/Textures/Default/Noise/LowResBlurredNoiseNormalized
  }
  NamedRerouteDeclaration(None=@n48.R) {name=GlowNoise_Breakup_02, nodeColor=<rgba 0.8281 1 0 1>}
  n49 = MaterialFunctionCall() {materialFunction=/Game/Global/Materials/MaterialGym/Materials/MaterialElements/UVs/CameraAlignedVirutalPlane/MF_CameraAlignedVirtualPlane_Wrap}
  n50 = SphereMask(
   A=@n49.UV
   B=AppendVector(A=Constant() {r=0.5}, B=ScalarParameter() {defaultValue=0.56, group=RampCanal, parameterName=ChestGlowPostion})
   Radius=Constant() {r=0.4}
   Hardness=Constant()
  )
  n51 = Constant() {r=1.4}
  NamedRerouteDeclaration(None=Power(Base=Multiply(A=@n50, B=@n51) {constB=1.5}, Exp=Constant() {r=6.7491})) {name=Heart GlowMAsk, nodeColor=<rgba 0 0.3672 1 1>}
  n52 = OneMinus(None=Fresnel(ExponentIn=Constant() {r=0.24}))
  n53 = ScalarParameter() {defaultValue=4, group=RampCanal, parameterName=Arm_InnerGlowIntensity}
  n54 = MaterialFunctionCall!(
   A=VertexInterpolator(VS=LocalPosition() {localOrigin=InstancePreSkinning}.XYZ).PS
   B=Constant4Vector() {constant=<rgba 0 0 0 1>}.RGB
   Bounds=Constant4Vector() {constant=<rgba 33.1459 86.3318 528.9321 1>}.RGB
   Edge Falloff=Constant() {r=13.5542}
  ) {materialFunction=/Engine/Functions/Engine_MaterialFunctions01/Gradient/BoxMask-3D}
  NamedRerouteDeclaration(None=Multiply(A=Multiply(A=@n52, B=@n53), B=OneMinus(None=@n54.Result))) {name=ArmMask, nodeColor=<rgba 1 0 0.8672 1>}
  n55 = TextureObjectParameter() {group=RampCanal, parameterName=TAMask, samplerType=SAMPLERTYPE_Masks, texture=&T_BlackMask}
  n56 = TextureSample(UVs=TextureCoordinate(), Tex=NamedRerouteDeclaration(None=@n55) {name=TA_Mask, nodeColor=<rgba 0.1719 1 0 1>}) {samplerSource=SSM_Wrap_WorldGroupSettings, samplerType=SAMPLERTYPE_Masks}
  NamedRerouteDeclaration(None=Saturate(None=Multiply(A=@n56.R) {constB=2})) {name=EmissiveMask, nodeColor=<rgba 1 0 0.8906 1>}
  n57 = ScalarParameter() {defaultValue=4, group=RampCanal, parameterName=Noise_Emissive_Tiling}
  n58 = TextureSample(UVs=Panner(Coordinate=Multiply(A=@n46.UV, B=@n57) {constB=11}, Speed=Constant() {r=0.05}) {speedX=0.07, speedY=0.07}) {
   samplerSource=SSM_Wrap_WorldGroupSettings
   samplerType=SAMPLERTYPE_Grayscale
   texture=&LowResBlurredNoiseNormalized
  }
  NamedRerouteDeclaration(None=@n58.R) {name=GlowNoise_Breakup_01, nodeColor=<rgba 0 1 0.7891 1>}
  NamedRerouteDeclaration(None=@n56.G) {name=EmissiveDetailMask, nodeColor=<rgba 1 0.4688 0 1>}
  NamedRerouteDeclaration(None=MaterialFunctionCall() {materialFunction=&MFW_CameraAlignedVirtualPlane}.UV) {name=CameraAlignedVirtualPlane, nodeColor=<rgba 1 0 0.6562 1>}
 ]
}
