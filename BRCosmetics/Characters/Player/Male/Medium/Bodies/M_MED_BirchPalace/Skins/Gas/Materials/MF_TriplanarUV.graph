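# Material function for the BirchPalace Gas skin that outputs side, top and front planar projection UVs from pre-skinned local position, leaving blending to the caller.
# A vertex interpolator passes the position, which is remapped against a box built from center and scale inputs, with optional clamping.
# Inverted YZ, XY and XZ pairs form the UVs.
# FlipUV lets the side projection mirror horizontally by which side of the center X lies on.
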MaterialFunction /BRCosmetics/Characters/Player/Male/Medium/Bodies/M_MED_BirchPalace/Skins/Gas/Materials/MF_TriplanarUV
graph {nodes [
  n1 = VertexInterpolator(VS=LocalPosition() {localOrigin=InstancePreSkinning}.XYZ)
  MaterialExpressionFunctionInput_6 = FunctionInput() {bUsePreviewValueAsDefault=true, inputName=TriplanarUV_CenterPosition}
  MaterialExpressionFunctionInput_0 = FunctionInput(Preview=Constant() {r=30}) {bUsePreviewValueAsDefault=true, inputName=TriplanarUV_UVScale, inputType=FunctionInput_Scalar}
  n2 = Custom!(Center=@MaterialExpressionFunctionInput_6, Size=@MaterialExpressionFunctionInput_0) {
   code="// Calculate half the length\r\n    float halfSize = Size * 0.5;\r\n    float3 V00;      // UV00\r\n    V00 = Center + float3(-halfSize, -halfSize, -halfSize);\r\n    return V00;"
   description=Get the 00 position
   showCode=true
   inputs [
    {
     inputName=Center
     input {expression=&self:MaterialExpressionFunctionInput_6}
    }
    {
     inputName=Size
     input {expression=&self:MaterialExpressionFunctionInput_0}
    }
   ]
  }
  n3 = Divide(A=Subtract(A=@n1.PS, B=@n2), B=@MaterialExpressionFunctionInput_0)
  n4 = LinearInterpolate(
   A=@n3
   B=Clamp(None=@n3)
   Alpha=FunctionInput(Preview=Constant() {r=1}) {bUsePreviewValueAsDefault=true, inputName=TriplanarUV_UseClamp?, inputType=FunctionInput_Scalar}
  )
  n5 = OneMinus(None=ComponentMask(None=@n4) {b=true, g=true})
  n6 = OneMinus(None=ComponentMask(None=@n5) {r=true})
  n7 = ComponentMask(None=@n5) {g=true}
  n8 = AppendVector(A=OneMinus(None=ComponentMask(None=@n5) {r=true}), B=ComponentMask(None=@n5) {g=true})
  n9 = Constant() {r=1}
  n10 = If(
   A=Subtract(A=ComponentMask(None=@n1.PS) {r=true}, B=ComponentMask(None=@MaterialExpressionFunctionInput_6) {r=true})
   B=Constant()
   "A > B"=@n9
   "A == B"=@n9
   "A < B"=Constant()
  )
  n11 = FunctionInput() {bUsePreviewValueAsDefault=true, inputName=TriplanarUV_FlipUV, inputType=FunctionInput_Scalar, previewValue=<0 0 0 0>}
  FunctionOutput(None=LinearInterpolate(A=AppendVector(A=@n6, B=@n7), B=LinearInterpolate(A=@n8, B=@n5, Alpha=@n10), Alpha=@n11)) {outputName=UV_SideProjection}
  FunctionOutput(None=OneMinus(None=ComponentMask(None=@n4) {g=true, r=true})) {outputName=UV_TopProjection}
  FunctionOutput(None=OneMinus(None=ComponentMask(None=@n4) {b=true, r=true})) {outputName=UV_FrontProjection}
  TransformPosition(None=@MaterialExpressionFunctionInput_6) {transformType=TRANSFORMPOSSOURCE_World}
 ]}
