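# MFW_TriplanarUV is a material function wrapper around MF_TriplanarUV in the materials folder of the BirchPalace Gas character skin.
# It feeds the inner function named parameters in the TriplanarUV group for clamping, UV flip, center position and UV size.
# It passes the side, top and front projection UVs out as separate X, Z and Y outputs, leaving any sampling or blending to the caller.
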MaterialFunction /BRCosmetics/Characters/Player/Male/Medium/Bodies/M_MED_BirchPalace/Skins/Gas/Materials/MFW_TriplanarUV
graph {nodes [
  n1 = MaterialFunctionCall!(
   TriplanarUV_UseClamp?=ScalarParameter() {group=TriplanarUV, parameterName=TriplanarUV_3D_UseClamp}
   TriplanarUV_FlipUV=ScalarParameter() {group=TriplanarUV, parameterName=TriplanarUV_3D_FlipUV}
   TriplanarUV_CenterPosition=VectorParameter() {group=TriplanarUV, parameterName=TriplanarUV_3D_CenterPosition, defaultValue=<rgba 0 0 0 1>}.RGB
   TriplanarUV_UVScale=ScalarParameter() {defaultValue=30, group=TriplanarUV, parameterName=TriplanarUV_3D_UVSize}
  ) {materialFunction=&/MF_TriplanarUV}
  FunctionOutput(None=@n1.UV_SideProjection) {outputName=Out_Projection_X}
  FunctionOutput(None=@n1.UV_TopProjection) {outputName=Out_Projection_Z}
  FunctionOutput(None=@n1.UV_FrontProjection) {outputName=Out_Projection_Y}
 ]}
