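# Material function that builds triplanar axis blend masks for the Gas skin of the BirchPalace character body, outputting X and Z masks only.
# It interpolates the pre-skinned normal from the vertex shader, so the masks stay fixed to the mesh during animation.
# Each absolute axis component is multiplied by the AxisBlendPower parameter, raised to AxisBlendSharpness, then saturated.
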MaterialFunction /BRCosmetics/Characters/Player/Male/Medium/Bodies/M_MED_BirchPalace/Skins/Gas/Materials/MF_TriplanarBlendMask
graph {nodes [
  n1 = NamedRerouteDeclaration(None=VertexInterpolator(VS=PreSkinnedNormal()).PS) {name=NormalLS, nodeColor=<rgba 1 0.7031 0 1>}
  n2 = ComponentMask(None=@n1) {b=true}
  n3 = ScalarParameter() {defaultValue=1.4359, group=TriplanarBlendMask, parameterName=TriplanarBlendMask_AxisBlendPower}
  n4 = ScalarParameter() {defaultValue=2, group=TriplanarBlendMask, parameterName=TriplanarBlendMask_AxisBlendSharpness}
  n5 = Power(Base=Multiply(A=Abs(None=@n2), B=@n3) {constB=2}, Exp=@n4)
  FunctionOutput(None=NamedRerouteDeclaration(None=Saturate(None=@n5)) {name=AxisMask_Z, nodeColor=<rgba 0.875 0 1 1>}) {outputName=AxisMask_Z}
  n6 = Abs(None=ComponentMask(None=@n1) {r=true})
  n7 = ComponentMask(None=Power(Base=Multiply(A=@n6, B=@n3) {constB=2}, Exp=@n4)) {r=true}
  FunctionOutput(None=NamedRerouteDeclaration(None=Saturate(None=@n7)) {name=AxisMask_X, nodeColor=<rgba 0.4062 1 0 1>}) {outputName=AxisMask_X}
 ]}
