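# Parent cloth-shaded head material for the BirchPalace Gas character skin (RampCanal), layering glowing smoke effects onto the 2023 anime character shader.
# Panned noise and triplanar flowing-smoke functions, borrowed partly from the RampCanal glider, build smoke layers.
# Sphere masks on pre-skinned local position localize head, forehead and neck glows independently of animation.
# Texture masks, Fresnel and eye glow shape the emissive, while lit and shaded colors get blended adjustments.
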Material /BRCosmetics/Characters/Player/Male/Medium/Bodies/M_MED_BirchPalace/Skins/Gas/Materials/M_RampCanal_Head_Parent
graph {
 properties {
  bAutomaticallySetUsageInEditor=false
  bUseMaterialAttributes=true
  bUsedWithMeshDeformer=true
  bUsedWithMorphTargets=true
  bUsedWithSkeletalMesh=true
  pixelDepthOffsetMode=PDOM_Legacy
  shadingModel=MSM_Cloth
  usedShadingModels=Cloth
 }
 nodes [
  n1 = MaterialFunctionCall!(CustomUV) {materialFunction=/Game/Global/Materials/Character_Parent_2023/Anime_Parent_2023/MaterialFunctions/MFC_AnimeShader_2023_Pre}
  n2 = MaterialFunctionCall!(CustomLightVector, OverrideNormal_WS, STT=@n1.STT, Normal_Map=@n1.STT, DFL=@n1.STT, FX=@n1.STT) {materialFunction=/Game/Global/Materials/Character_Parent_2023/Anime_Parent_2023/MaterialFunctions/MFC_AnimeShader_2023_Calculations}
  n3 = NamedRerouteUsage()
  n4 = LinearInterpolate(
   A=ScalarParameter() {defaultValue=0.3, group=RampCanal, parameterName=BaseColor_BlendWeight}
   B=Constant() {r=0.7}
   Alpha=NamedRerouteUsage()
  )
  n5 = NamedRerouteUsage()
  MaterialFunctionCall!(
   Shading_Gradient=@n2.Shading_Gradient
   Threshold_Gradient=@n2.Shading_Gradient
   Specular_Gradient=@n2.Shading_Gradient
   Sheen_Mask=@n2.Shading_Gradient
   Thickness=@n2.Shading_Gradient
   DFL_Mask=@n2.Shading_Gradient
   FX=@n2.Shading_Gradient
   Metallic_Gradient=@n2.Shading_Gradient
   RimLight_Gradient=@n2.Shading_Gradient
   NormalBuffer=@n2.Shading_Gradient
   --------
   CINE_LightSources=@n2.Shading_Gradient
   CINE_LightColor=@n2.Shading_Gradient
   ---------
   Color_Lit=LinearInterpolate(A=NamedRerouteUsage(), B=Multiply(A=@n3, B=@n4) {constB=0.2}, Alpha=NamedRerouteUsage()) {constB=0}
   Color_Shaded=LinearInterpolate(A=NamedRerouteUsage(), B=Multiply(A=@n5, B=@n4) {constB=0.2}, Alpha=NamedRerouteUsage()) {constB=0}
   SRM=@n1.SRM
   Emissive=NamedRerouteUsage()
   DFL_Color=NamedRerouteDeclaration(None=@n1.DFL_Color) {name=DFL_Color, nodeColor=<rgba 1 0 0.9844 1>}
   Specular_Color=@n1.SRM
   RimLight_Color=@n1.SRM
  ) {materialFunction=/Game/Global/Materials/Character_Parent_2023/Anime_Parent_2023/MaterialFunctions/MFC_AnimeShader_2023_Post}
  NamedRerouteDeclaration(None=@n1.Color_Shaded) {name=Color_Shaded, nodeColor=<rgba 1 0.9141 0 1>}
  n6 = Constant() {r=0.392}
  n7 = Power(Base=OneMinus(None=Fresnel(ExponentIn=@n6)), Exp=Constant() {r=2}) {constExponent=3}
  n8 = SphereMask(
   A=NamedRerouteUsage()
   B=Constant4Vector() {constant=<rgba 0 -3.8825 147.4558 1>}.RGB
   Radius=Constant() {r=6}
   Hardness=Constant() {r=0.563}
  )
  n9 = Constant() {r=0.7364}
  n10 = Power(Base=OneMinus(None=Fresnel(ExponentIn=@n9)), Exp=Constant() {r=3}) {constExponent=3}
  n11 = SphereMask(
   A=NamedRerouteUsage()
   B=Constant4Vector() {constant=<rgba 0 -3.8825 147.4558 1>}.RGB
   Radius=Constant() {r=13.8698}
   Hardness=Constant() {r=0.563}
  )
  n12 = Add(
   A=Multiply(A=Multiply(A=@n7, B=@n8), B=Constant() {r=10}) {constB=2}
   B=Multiply(A=Multiply(A=@n10, B=@n11), B=Constant() {r=1}) {constB=0.5}
  )
  n13 = NamedRerouteUsage()
  NamedRerouteDeclaration(None=Saturate(None=Add(A=@n12, B=@n13))) {name=NeckInnterGlowMask, nodeColor=<rgba 0 1 0.2734 1>}
  NamedRerouteDeclaration(None=VertexInterpolator(VS=LocalPosition() {localOrigin=InstancePreSkinning}.XYZ).PS) {name=Pre_SkinnedLocalPostion, nodeColor=<rgba 1 0.7031 0 1>}
  NamedRerouteDeclaration(None=@n1.Color_Lit) {name=Color_Lit, nodeColor=<rgba 0.6641 1 0 1>}
  n14 = Constant() {r=1}
  n15 = SphereMask(A=NamedRerouteUsage(), B=Constant4Vector() {constant=<rgba 0 0 151.8055 1>}.RGB, Radius=Constant() {r=15}, Hardness=Constant() {r=0.5})
  n16 = Multiply(A=NamedRerouteUsage()) {constB=1.5}
  n17 = LinearInterpolate(
   A=ScalarParameter() {defaultValue=0.75, group=RampCanal, parameterName=Noise_EmissiveIntensity_Min}
   B=LinearInterpolate(
    A=Constant() {r=1}
    B=ScalarParameter() {defaultValue=50, group=RampCanal, parameterName=Noise_EmissiveIntensity_Max}
    Alpha=NamedRerouteDeclaration(None=@n15) {name=HeadMask, nodeColor=<rgba 1 0 0.4219 1>}
   )
   Alpha=Saturate(None=Power(Base=@n16) {constExponent=1})
  ) {constA=1, constB=2}
  n18 = ScalarParameter() {group=RampCanal, parameterName=EmissiveDetailMaskIntensity_MIn}
  n19 = ScalarParameter() {defaultValue=3, group=RampCanal, parameterName=EmissiveDetailMaskIntensity_Max}
  n20 = NamedRerouteUsage()
  NamedRerouteDeclaration(None=Multiply(A=Multiply(A=@n14, B=@n17), B=LinearInterpolate(A=@n18, B=@n19, Alpha=@n20) {constA=1, constB=2})) {name=RampCanalBaseSmoke, nodeColor=<rgba 0 0.9062 1 1>}
  NamedRerouteDeclaration(None=@n1.Emissive) {name=In_Emissive, nodeColor=<rgba 1 0.7734 0 1>}
  n21 = MaterialFunctionCall() {materialFunction=&/MFW_TriplanarUV}
  NamedRerouteDeclaration(None=@n21.Out_Projection_X) {name=Projection_X, nodeColor=<rgba 1 0 0.6562 1>}
  n22 = ScalarParameter() {defaultValue=0.1, group=RampCanal, parameterName=FireSmokelEmissiveIntensity_Min}
  n23 = ScalarParameter() {defaultValue=0.09, group=RampCanal, parameterName=FireSmokelEmissiveIntensity_Min2}
  n24 = NamedRerouteUsage()
  n25 = Constant() {r=0.5}
  n26 = ScalarParameter() {defaultValue=3, group=RampCanal, parameterName=FireSmokeEmissiveIntensity_Max}
  n27 = NamedRerouteUsage()
  n28 = LinearInterpolate(
   A=MaterialFunctionCall!(In_FireSmoke_TAMask=NamedRerouteUsage(), In_FireSmoke_UV=NamedRerouteUsage()) {materialFunction=/BRCosmetics/Gadgets/Assets/VinderTech_GliderChute/Glider_RampCanal/Materials/MFW_FlowingSmoke}.Result
   B=MaterialFunctionCall(In_FireSmoke_TAMask=NamedRerouteUsage(), In_FireSmoke_UV=NamedRerouteUsage()) {materialFunction=&MFW_FlowingSmoke}.Result
   Alpha=MaterialFunctionCall() {materialFunction=&/MF_TriplanarBlendMask}.AxisMask_X
  )
  n29 = NamedRerouteUsage()
  NamedRerouteDeclaration(None=LinearInterpolate(A=LinearInterpolate(A=@n22, B=@n23, Alpha=@n24), B=LinearInterpolate(A=@n25, B=@n26, Alpha=@n27) {constA=1}, Alpha=Multiply(A=@n28, B=@n29)) {constA=1, constB=2}) {name=DetailSmoke, nodeColor=<rgba 0 1 0.1328 1>}
  n30 = Multiply(
   A=LinearInterpolate(
    A=Constant4Vector() {constant=<rgba 0.09 0.1267 0.2 1>}.RGB
    B=Constant4Vector() {constant=<rgba 0.081 0.2878 0.8 1>}.RGB
    Alpha=NamedRerouteUsage()
   )
   B=LinearInterpolate(A=Constant() {r=10}, B=NamedRerouteUsage(), Alpha=NamedRerouteUsage()) {constA=10}
  )
  n31 = LinearInterpolate(Alpha=NamedRerouteUsage()) {constA=1, constB=0.05}
  n32 = Multiply(A=LinearInterpolate(A=NamedRerouteUsage(), B=Multiply(A=@n30, B=@n31), Alpha=NamedRerouteUsage()), B=NamedRerouteUsage())
  n33 = NamedRerouteUsage()
  n34 = TextureSample() {
   samplerSource=SSM_Wrap_WorldGroupSettings
   samplerType=SAMPLERTYPE_Masks
   texture=/BRCosmetics/Characters/Player/Male/Medium/Bodies/M_MED_BirchPalace/Skins/Gas/Textures/T_BirchPalace_Gas_Body_TA2
  }
  n35 = Multiply(
   A=LinearInterpolate(A=@n32, B=Multiply(A=@n32, B=@n33), Alpha=ScalarParameter() {defaultValue=0.75, group=RampCanal, parameterName=FresnelBrightness})
   B=LinearInterpolate(B=ScalarParameter() {defaultValue=5, group=RampCanal, parameterName=EyeGlowIntensity}, Alpha=@n34.G) {constA=1, constB=20}
  )
  n36 = Constant4Vector() {constant=<rgba 0.0331 0.0742 0.1981 1>}
  n37 = LinearInterpolate(A=Multiply(A=Constant() {r=0.3}, B=NamedRerouteUsage()), Alpha=NamedRerouteUsage()) {constB=0.05}
  n38 = Max(A=OneMinus(None=NamedRerouteUsage()), B=NamedRerouteUsage())
  n39 = NamedRerouteUsage()
  NamedRerouteDeclaration(None=LinearInterpolate(
    A=VectorParameter() {group=RampCanal, parameterName=FresnelGlowColor, defaultValue=<rgba 0.0502 0.0537 0.22 1>}.RGB
    B=LinearInterpolate(A=@n35, B=@n36.RGB, Alpha=@n37)
    Alpha=LinearInterpolate(B=@n38, Alpha=@n39) {constA=1}
   )) {name=Out_Emissive, nodeColor=<rgba 1 0 0.1875 1>}
  n40 = Constant() {r=1}
  n41 = Constant() {r=-0.5}
  n42 = Saturate(None=Fresnel(ExponentIn=Constant() {r=1}) {exponent=1})
  NamedRerouteDeclaration(None=Saturate(None=LinearInterpolate(A=@n40, B=@n41, Alpha=@n42))) {name=Fresnel, nodeColor=<rgba 1 0 0.4219 1>}
  NamedRerouteDeclaration(None=ComponentMask(None=@n1.STT) {g=true}) {name=AO, nodeColor=<rgba 0.1719 1 0 1>}
  n43 = NamedRerouteUsage()
  n44 = Constant() {r=5}
  n45 = TextureSample(UVs=Panner(Coordinate=Multiply(A=@n43, B=@n44) {constB=11}, Speed=Constant() {r=0.1}) {speedX=0.07, speedY=0.07}) {
   samplerSource=SSM_Wrap_WorldGroupSettings
   samplerType=SAMPLERTYPE_Grayscale
   texture=/Game/Global/Textures/Default/Noise/LowResBlurredNoiseNormalized
  }
  NamedRerouteDeclaration(None=@n45.R) {name=GlowNoise_Breakup_02, nodeColor=<rgba 0.8281 1 0 1>}
  n46 = SphereMask(A=NamedRerouteUsage(), B=Constant4Vector() {constant=<rgba 0 6.5078 158 1>}.RGB, Radius=Constant() {r=12}, Hardness=Constant() {r=-0.488})
  NamedRerouteDeclaration(None=Power(Base=Multiply(A=@n46) {constB=2})) {name=ForeheadCore, nodeColor=<rgba 0.1016 1 0 1>}
  n47 = TextureObjectParameter() {group=RampCanal, parameterName=TAMask, samplerType=SAMPLERTYPE_Masks, texture=&T_BlackMask}
  n48 = TextureSample(UVs=TextureCoordinate(), Tex=NamedRerouteDeclaration(None=@n47) {name=TA_Mask, nodeColor=<rgba 0.1719 1 0 1>}) {samplerSource=SSM_Wrap_WorldGroupSettings, samplerType=SAMPLERTYPE_Masks}
  NamedRerouteDeclaration(None=Saturate(None=Multiply(A=@n48.R) {constB=2})) {name=EmissiveMask, nodeColor=<rgba 1 0 0.8906 1>}
  n49 = NamedRerouteUsage()
  n50 = ScalarParameter() {defaultValue=2, group=RampCanal, parameterName=Noise_Emissive_Tiling}
  n51 = TextureSample(UVs=Panner(Coordinate=Multiply(A=@n49, B=@n50) {constB=11}, Speed=Constant() {r=0.1}) {speedX=0.07, speedY=0.07}) {
   samplerSource=SSM_Wrap_WorldGroupSettings
   samplerType=SAMPLERTYPE_Grayscale
   texture=&LowResBlurredNoiseNormalized
  }
  NamedRerouteDeclaration(None=@n51.R) {name=GlowNoise_Breakup_01, nodeColor=<rgba 0 1 0.7891 1>}
  NamedRerouteDeclaration(None=@n48.G) {name=EmissiveDetailMask, nodeColor=<rgba 1 0.4688 0 1>}
  NamedRerouteDeclaration(None=@n21.Out_Projection_Y) {name=Projection_Y, nodeColor=<rgba 1 0 0.6562 1>}
 ]
}
